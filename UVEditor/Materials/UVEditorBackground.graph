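# Unlit, two-sided material for the UV Editor background, offering one texture parameter per sampler type with color-grid defaults, plus a pixel depth offset parameter.
# A "Texture Coordinate Selector" composite switches the UVChannel parameter among eight UV channels to feed every sampler.
# A second Switch uses TextureSampler to pick among nine typed samples or constant black, working around fixed per-node sampler types.
# The External texture sampler is an unconnected TODO.
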Material /UVEditor/Materials/UVEditorBackground
graph {
 expressions {
  MaterialExpressionComposite_0 {
   class=MaterialExpressionComposite
   inputs [Channel {from=MaterialExpressionScalarParameter_4, output=""}]
   properties {desc=Converts a supplied channel number into UV coordinates from the mesh's attribute data., subgraphName=Texture Coordinate Selector}
  }
  MaterialExpressionConstant3Vector_0 {class=MaterialExpressionConstant3Vector}
  MaterialExpressionPinBase_0 {
   class=MaterialExpressionPinBase
   inputs []
   properties {reroutePins [Channel {expression=&self:MaterialGraph_0.MaterialGraphNode_Composite_0.Texture Coordinate Selector.MaterialExpressionReroute_2}]}
  }
  MaterialExpressionPinBase_1 {
   class=MaterialExpressionPinBase
   inputs [Texture Coordinates {from=MaterialExpressionSwitch_1, output=""}]
   properties {reroutePins [Texture Coordinates {expression=&self:MaterialGraph_0.MaterialGraphNode_Composite_0.Texture Coordinate Selector.MaterialExpressionReroute_3}]}
  }
  MaterialExpressionScalarParameter_0 {class=MaterialExpressionScalarParameter, properties {defaultValue=-1, parameterName=BackgroundPixelDepthOffset}}
  MaterialExpressionScalarParameter_2 {class=MaterialExpressionScalarParameter, properties {parameterName=TextureSampler, sliderMax=1}}
  MaterialExpressionScalarParameter_4 {class=MaterialExpressionScalarParameter, properties {parameterName=UVChannel, sliderMax=8}}
  MaterialExpressionSwitch_1 {
   class=MaterialExpressionSwitch
   inputs [
    SwitchValue {from=MaterialExpressionReroute_2, output=Channel}
    Default {from="", output=""}
    None {from=MaterialExpressionTextureCoordinate_16, output=""}
    None {from=MaterialExpressionTextureCoordinate_17, output=""}
    None {from=MaterialExpressionTextureCoordinate_18, output=""}
    None {from=MaterialExpressionTextureCoordinate_19, output=""}
    None {from=MaterialExpressionTextureCoordinate_20, output=""}
    None {from=MaterialExpressionTextureCoordinate_21, output=""}
    None {from=MaterialExpressionTextureCoordinate_22, output=""}
    None {from=MaterialExpressionTextureCoordinate_23, output=""}
   ]
   properties {inputs [
     {
      inputName=None
      input {
       expression=&self:MaterialExpressionTextureCoordinate_16
       inputName=None
       mask=0
       maskA=0
       maskB=0
       maskG=0
       maskR=0
       outputIndex=0
      }
     }
     {
      inputName=None
      input {
       expression=&self:MaterialExpressionTextureCoordinate_17
       inputName=None
       mask=0
       maskA=0
       maskB=0
       maskG=0
       maskR=0
       outputIndex=0
      }
     }
     {
      inputName=None
      input {
       expression=&self:MaterialExpressionTextureCoordinate_18
       inputName=None
       mask=0
       maskA=0
       maskB=0
       maskG=0
       maskR=0
       outputIndex=0
      }
     }
     {
      inputName=None
      input {
       expression=&self:MaterialExpressionTextureCoordinate_19
       inputName=None
       mask=0
       maskA=0
       maskB=0
       maskG=0
       maskR=0
       outputIndex=0
      }
     }
     {
      inputName=None
      input {
       expression=&self:MaterialExpressionTextureCoordinate_20
       inputName=None
       mask=0
       maskA=0
       maskB=0
       maskG=0
       maskR=0
       outputIndex=0
      }
     }
     {
      inputName=None
      input {
       expression=&self:MaterialExpressionTextureCoordinate_21
       inputName=None
       mask=0
       maskA=0
       maskB=0
       maskG=0
       maskR=0
       outputIndex=0
      }
     }
     {
      inputName=None
      input {
       expression=&self:MaterialExpressionTextureCoordinate_22
       inputName=None
       mask=0
       maskA=0
       maskB=0
       maskG=0
       maskR=0
       outputIndex=0
      }
     }
     {
      inputName=None
      input {
       expression=&self:MaterialExpressionTextureCoordinate_23
       inputName=None
       mask=0
       maskA=0
       maskB=0
       maskG=0
       maskR=0
       outputIndex=0
      }
     }
    ]}
  }
  MaterialExpressionSwitch_2 {
   class=MaterialExpressionSwitch
   inputs [
    SwitchValue {from=MaterialExpressionScalarParameter_2, output=""}
    Default {from="", output=""}
    None {from=MaterialExpressionTextureSampleParameter2D_2, output=RGB}
    None {from=MaterialExpressionTextureSampleParameter2D_1, output=RGB}
    None {from=MaterialExpressionTextureSampleParameter2D_3, output=RGB}
    None {from=MaterialExpressionTextureSampleParameter2D_4, output=RGB}
    None {from=MaterialExpressionTextureSampleParameter2D_5, output=RGB}
    None {from=MaterialExpressionTextureSampleParameter2D_6, output=RGB}
    None {from=MaterialExpressionTextureSampleParameter2D_7, output=RGB}
    None {from=MaterialExpressionTextureSampleParameter2D_8, output=RGB}
    None {from=MaterialExpressionTextureSampleParameter2D_9, output=RGB}
    None {from=MaterialExpressionConstant3Vector_0, output=""}
   ]
   properties {inputs [
     {
      inputName=None
      input {
       expression=&self:MaterialExpressionTextureSampleParameter2D_2
       inputName=None
       mask=1
       maskA=0
       maskB=1
       maskG=1
       maskR=1
       outputIndex=0
      }
     }
     {
      inputName=None
      input {
       expression=&self:MaterialExpressionTextureSampleParameter2D_1
       inputName=None
       mask=1
       maskA=0
       maskB=1
       maskG=1
       maskR=1
       outputIndex=0
      }
     }
     {
      inputName=None
      input {
       expression=&self:MaterialExpressionTextureSampleParameter2D_3
       inputName=None
       mask=1
       maskA=0
       maskB=1
       maskG=1
       maskR=1
       outputIndex=0
      }
     }
     {
      inputName=None
      input {
       expression=&self:MaterialExpressionTextureSampleParameter2D_4
       inputName=None
       mask=1
       maskA=0
       maskB=1
       maskG=1
       maskR=1
       outputIndex=0
      }
     }
     {
      inputName=None
      input {
       expression=&self:MaterialExpressionTextureSampleParameter2D_5
       inputName=None
       mask=1
       maskA=0
       maskB=1
       maskG=1
       maskR=1
       outputIndex=0
      }
     }
     {
      inputName=None
      input {
       expression=&self:MaterialExpressionTextureSampleParameter2D_6
       inputName=None
       mask=1
       maskA=0
       maskB=1
       maskG=1
       maskR=1
       outputIndex=0
      }
     }
     {
      inputName=None
      input {
       expression=&self:MaterialExpressionTextureSampleParameter2D_7
       inputName=None
       mask=1
       maskA=0
       maskB=1
       maskG=1
       maskR=1
       outputIndex=0
      }
     }
     {
      inputName=None
      input {
       expression=&self:MaterialExpressionTextureSampleParameter2D_8
       inputName=None
       mask=1
       maskA=0
       maskB=1
       maskG=1
       maskR=1
       outputIndex=0
      }
     }
     {
      inputName=None
      input {
       expression=&self:MaterialExpressionTextureSampleParameter2D_9
       inputName=None
       mask=1
       maskA=0
       maskB=1
       maskG=1
       maskR=1
       outputIndex=0
      }
     }
     {
      inputName=None
      input {
       expression=&self:MaterialExpressionConstant3Vector_0
       inputName=None
       mask=1
       maskA=0
       maskB=1
       maskG=1
       maskR=1
       outputIndex=0
      }
     }
    ]}
  }
  MaterialExpressionTextureCoordinate_16 {class=MaterialExpressionTextureCoordinate}
  MaterialExpressionTextureCoordinate_17 {class=MaterialExpressionTextureCoordinate, properties {coordinateIndex=1}}
  MaterialExpressionTextureCoordinate_18 {class=MaterialExpressionTextureCoordinate, properties {coordinateIndex=2}}
  MaterialExpressionTextureCoordinate_19 {class=MaterialExpressionTextureCoordinate, properties {coordinateIndex=3}}
  MaterialExpressionTextureCoordinate_20 {class=MaterialExpressionTextureCoordinate, properties {coordinateIndex=4}}
  MaterialExpressionTextureCoordinate_21 {class=MaterialExpressionTextureCoordinate, properties {coordinateIndex=5}}
  MaterialExpressionTextureCoordinate_22 {class=MaterialExpressionTextureCoordinate, properties {coordinateIndex=6}}
  MaterialExpressionTextureCoordinate_23 {class=MaterialExpressionTextureCoordinate, properties {coordinateIndex=7}}
  MaterialExpressionTextureSampleParameter2D_1 {
   class=MaterialExpressionTextureSampleParameter2D
   inputs [
    UVs {from=MaterialExpressionReroute_3, output=Texture Coordinates}
    Apply View MipBias {from="", output=""}
   ]
   properties {parameterName=BackgroundBaseMap_Grayscale, samplerType=SAMPLERTYPE_Grayscale, texture=/UVEditor/Textures/UVEditorColorGrid_Grayscale}
  }
  MaterialExpressionTextureSampleParameter2D_2 {
   class=MaterialExpressionTextureSampleParameter2D
   inputs [
    UVs {from=MaterialExpressionReroute_3, output=Texture Coordinates}
    Apply View MipBias {from="", output=""}
   ]
   properties {parameterName=BackgroundBaseMap_Color, texture=/UVEditor/Textures/UVEditorColorGrid_Color}
  }
  MaterialExpressionTextureSampleParameter2D_21 {
   class=MaterialExpressionTextureSampleParameter2D
   inputs [
    UVs {from=MaterialExpressionReroute_3, output=Texture Coordinates}
    Apply View MipBias {from="", output=""}
   ]
   properties {
    desc=TODO: Add support for this texture type.
    parameterName=BackgroundBaseMap_External
    samplerType=SAMPLERTYPE_Data
    texture=/UVEditor/Textures/UVEditorColorGrid_Data
   }
  }
  MaterialExpressionTextureSampleParameter2D_3 {
   class=MaterialExpressionTextureSampleParameter2D
   inputs [
    UVs {from=MaterialExpressionReroute_3, output=Texture Coordinates}
    Apply View MipBias {from="", output=""}
   ]
   properties {parameterName=BackgroundBaseMap_Alpha, samplerType=SAMPLERTYPE_Alpha, texture=/UVEditor/Textures/UVEditorColorGrid_Alpha}
  }
  MaterialExpressionTextureSampleParameter2D_4 {
   class=MaterialExpressionTextureSampleParameter2D
   inputs [
    UVs {from=MaterialExpressionReroute_3, output=Texture Coordinates}
    Apply View MipBias {from="", output=""}
   ]
   properties {parameterName=BackgroundBaseMap_Normal, samplerType=SAMPLERTYPE_Normal, texture=/UVEditor/Textures/UVEditorColorGrid_Normal}
  }
  MaterialExpressionTextureSampleParameter2D_5 {
   class=MaterialExpressionTextureSampleParameter2D
   inputs [
    UVs {from=MaterialExpressionReroute_3, output=Texture Coordinates}
    Apply View MipBias {from="", output=""}
   ]
   properties {parameterName=BackgroundBaseMap_Masks, samplerType=SAMPLERTYPE_Masks, texture=/UVEditor/Textures/UVEditorColorGrid_Mask}
  }
  MaterialExpressionTextureSampleParameter2D_6 {
   class=MaterialExpressionTextureSampleParameter2D
   inputs [
    UVs {from=MaterialExpressionReroute_3, output=Texture Coordinates}
    Apply View MipBias {from="", output=""}
   ]
   properties {
    parameterName=BackgroundBaseMap_DistanceFieldFont
    samplerType=SAMPLERTYPE_DistanceFieldFont
    texture=/UVEditor/Textures/UVEditorColorGrid_DistanceFieldFont
   }
  }
  MaterialExpressionTextureSampleParameter2D_7 {
   class=MaterialExpressionTextureSampleParameter2D
   inputs [
    UVs {from=MaterialExpressionReroute_3, output=Texture Coordinates}
    Apply View MipBias {from="", output=""}
   ]
   properties {parameterName=BackgroundBaseMap_LinearColor, samplerType=SAMPLERTYPE_LinearColor, texture=/UVEditor/Textures/UVEditorColorGrid_LinearColor}
  }
  MaterialExpressionTextureSampleParameter2D_8 {
   class=MaterialExpressionTextureSampleParameter2D
   inputs [
    UVs {from=MaterialExpressionReroute_3, output=Texture Coordinates}
    Apply View MipBias {from="", output=""}
   ]
   properties {
    parameterName=BackgroundBaseMap_LinearGrayscale
    samplerType=SAMPLERTYPE_LinearGrayscale
    texture=/UVEditor/Textures/UVEditorColorGrid_LinearGrayscale
   }
  }
  MaterialExpressionTextureSampleParameter2D_9 {
   class=MaterialExpressionTextureSampleParameter2D
   inputs [
    UVs {from=MaterialExpressionReroute_3, output=Texture Coordinates}
    Apply View MipBias {from="", output=""}
   ]
   properties {parameterName=BackgroundBaseMap_Data, samplerType=SAMPLERTYPE_Data, texture=&UVEditorColorGrid_Data}
  }
 }
 properties {pixelDepthOffsetMode=PDOM_Legacy, shadingModel=MSM_Unlit, twoSided=true, usedShadingModels=Unlit}
}
