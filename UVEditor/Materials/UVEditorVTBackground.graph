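# UV Editor background material that samples a virtual texture, defaulting to color grid textures.
# A UVChannel switch picks one of eight texture coordinate sets to feed seven texture parameters, one per virtual sampler type (color, grayscale, alpha, normal, masks, linear color, linear grayscale).
# A VirtualTextureSampler switch selects among their RGB results, so a single material covers each of these sampler types.
# A BackgroundPixelDepthOffset scalar is also exposed.
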Material /UVEditor/Materials/UVEditorVTBackground
graph {
 expressions {
  MaterialExpressionScalarParameter_0 {class=MaterialExpressionScalarParameter, properties {defaultValue=-1, parameterName=BackgroundPixelDepthOffset}}
  MaterialExpressionScalarParameter_1 {class=MaterialExpressionScalarParameter, properties {parameterName=UVChannel}}
  MaterialExpressionScalarParameter_2 {class=MaterialExpressionScalarParameter, properties {parameterName=VirtualTextureSampler, sliderMax=1}}
  MaterialExpressionSwitch_0 {
   class=MaterialExpressionSwitch
   inputs [
    SwitchValue {from=MaterialExpressionScalarParameter_1, output=""}
    Default {from="", output=""}
    None {from=MaterialExpressionTextureCoordinate_8, output=""}
    None {from=MaterialExpressionTextureCoordinate_9, output=""}
    None {from=MaterialExpressionTextureCoordinate_10, output=""}
    None {from=MaterialExpressionTextureCoordinate_11, output=""}
    None {from=MaterialExpressionTextureCoordinate_12, output=""}
    None {from=MaterialExpressionTextureCoordinate_13, output=""}
    None {from=MaterialExpressionTextureCoordinate_14, output=""}
    None {from=MaterialExpressionTextureCoordinate_15, output=""}
   ]
   properties {inputs [
     {
      inputName=None
      input {
       expression=&self:MaterialExpressionTextureCoordinate_8
       inputName=None
       mask=0
       maskA=0
       maskB=0
       maskG=0
       maskR=0
       outputIndex=0
      }
     }
     {
      inputName=None
      input {
       expression=&self:MaterialExpressionTextureCoordinate_9
       inputName=None
       mask=0
       maskA=0
       maskB=0
       maskG=0
       maskR=0
       outputIndex=0
      }
     }
     {
      inputName=None
      input {
       expression=&self:MaterialExpressionTextureCoordinate_10
       inputName=None
       mask=0
       maskA=0
       maskB=0
       maskG=0
       maskR=0
       outputIndex=0
      }
     }
     {
      inputName=None
      input {
       expression=&self:MaterialExpressionTextureCoordinate_11
       inputName=None
       mask=0
       maskA=0
       maskB=0
       maskG=0
       maskR=0
       outputIndex=0
      }
     }
     {
      inputName=None
      input {
       expression=&self:MaterialExpressionTextureCoordinate_12
       inputName=None
       mask=0
       maskA=0
       maskB=0
       maskG=0
       maskR=0
       outputIndex=0
      }
     }
     {
      inputName=None
      input {
       expression=&self:MaterialExpressionTextureCoordinate_13
       inputName=None
       mask=0
       maskA=0
       maskB=0
       maskG=0
       maskR=0
       outputIndex=0
      }
     }
     {
      inputName=None
      input {
       expression=&self:MaterialExpressionTextureCoordinate_14
       inputName=None
       mask=0
       maskA=0
       maskB=0
       maskG=0
       maskR=0
       outputIndex=0
      }
     }
     {
      inputName=None
      input {
       expression=&self:MaterialExpressionTextureCoordinate_15
       inputName=None
       mask=0
       maskA=0
       maskB=0
       maskG=0
       maskR=0
       outputIndex=0
      }
     }
    ]}
  }
  MaterialExpressionSwitch_1 {
   class=MaterialExpressionSwitch
   inputs [
    SwitchValue {from=MaterialExpressionScalarParameter_2, output=""}
    Default {from="", output=""}
    None {from=MaterialExpressionTextureSampleParameter2D_11, output=RGB}
    None {from=MaterialExpressionTextureSampleParameter2D_12, output=RGB}
    None {from=MaterialExpressionTextureSampleParameter2D_13, output=RGB}
    None {from=MaterialExpressionTextureSampleParameter2D_14, output=RGB}
    None {from=MaterialExpressionTextureSampleParameter2D_15, output=RGB}
    None {from=MaterialExpressionTextureSampleParameter2D_17, output=RGB}
    None {from=MaterialExpressionTextureSampleParameter2D_18, output=RGB}
   ]
   properties {inputs [
     {
      inputName=None
      input {
       expression=&self:MaterialExpressionTextureSampleParameter2D_11
       inputName=None
       mask=1
       maskA=0
       maskB=1
       maskG=1
       maskR=1
       outputIndex=0
      }
     }
     {
      inputName=None
      input {
       expression=&self:MaterialExpressionTextureSampleParameter2D_12
       inputName=None
       mask=1
       maskA=0
       maskB=1
       maskG=1
       maskR=1
       outputIndex=0
      }
     }
     {
      inputName=None
      input {
       expression=&self:MaterialExpressionTextureSampleParameter2D_13
       inputName=None
       mask=1
       maskA=0
       maskB=1
       maskG=1
       maskR=1
       outputIndex=0
      }
     }
     {
      inputName=None
      input {
       expression=&self:MaterialExpressionTextureSampleParameter2D_14
       inputName=None
       mask=1
       maskA=0
       maskB=1
       maskG=1
       maskR=1
       outputIndex=0
      }
     }
     {
      inputName=None
      input {
       expression=&self:MaterialExpressionTextureSampleParameter2D_15
       inputName=None
       mask=1
       maskA=0
       maskB=1
       maskG=1
       maskR=1
       outputIndex=0
      }
     }
     {
      inputName=None
      input {
       expression=&self:MaterialExpressionTextureSampleParameter2D_17
       inputName=None
       mask=1
       maskA=0
       maskB=1
       maskG=1
       maskR=1
       outputIndex=0
      }
     }
     {
      inputName=None
      input {
       expression=&self:MaterialExpressionTextureSampleParameter2D_18
       inputName=None
       mask=1
       maskA=0
       maskB=1
       maskG=1
       maskR=1
       outputIndex=0
      }
     }
    ]}
  }
  MaterialExpressionTextureCoordinate_10 {class=MaterialExpressionTextureCoordinate, properties {coordinateIndex=2}}
  MaterialExpressionTextureCoordinate_11 {class=MaterialExpressionTextureCoordinate, properties {coordinateIndex=3}}
  MaterialExpressionTextureCoordinate_12 {class=MaterialExpressionTextureCoordinate, properties {coordinateIndex=4}}
  MaterialExpressionTextureCoordinate_13 {class=MaterialExpressionTextureCoordinate, properties {coordinateIndex=5}}
  MaterialExpressionTextureCoordinate_14 {class=MaterialExpressionTextureCoordinate, properties {coordinateIndex=6}}
  MaterialExpressionTextureCoordinate_15 {class=MaterialExpressionTextureCoordinate, properties {coordinateIndex=7}}
  MaterialExpressionTextureCoordinate_8 {class=MaterialExpressionTextureCoordinate}
  MaterialExpressionTextureCoordinate_9 {class=MaterialExpressionTextureCoordinate, properties {coordinateIndex=1}}
  MaterialExpressionTextureSampleParameter2D_11 {
   class=MaterialExpressionTextureSampleParameter2D
   inputs [
    UVs {from=MaterialExpressionSwitch_0, output=""}
    Apply View MipBias {from="", output=""}
   ]
   properties {parameterName=BackgroundBaseMap_VTColor, samplerType=SAMPLERTYPE_VirtualColor, texture=/UVEditor/Textures/UVEditorColorGrid_VTColor}
  }
  MaterialExpressionTextureSampleParameter2D_12 {
   class=MaterialExpressionTextureSampleParameter2D
   inputs [
    UVs {from=MaterialExpressionSwitch_0, output=""}
    Apply View MipBias {from="", output=""}
   ]
   properties {
    parameterName=BackgroundBaseMap_VTGrayscale
    samplerType=SAMPLERTYPE_VirtualGrayscale
    texture=/UVEditor/Textures/UVEditorColorGrid_VTGrayscale
   }
  }
  MaterialExpressionTextureSampleParameter2D_13 {
   class=MaterialExpressionTextureSampleParameter2D
   inputs [
    UVs {from=MaterialExpressionSwitch_0, output=""}
    Apply View MipBias {from="", output=""}
   ]
   properties {parameterName=BackgroundBaseMap_VTAlpha, samplerType=SAMPLERTYPE_VirtualAlpha, texture=/UVEditor/Textures/UVEditorColorGrid_VTAlpha}
  }
  MaterialExpressionTextureSampleParameter2D_14 {
   class=MaterialExpressionTextureSampleParameter2D
   inputs [
    UVs {from=MaterialExpressionSwitch_0, output=""}
    Apply View MipBias {from="", output=""}
   ]
   properties {parameterName=BackgroundBaseMap_VTNormal, samplerType=SAMPLERTYPE_VirtualNormal, texture=/UVEditor/Textures/UVEditorColorGrid_VTNormal}
  }
  MaterialExpressionTextureSampleParameter2D_15 {
   class=MaterialExpressionTextureSampleParameter2D
   inputs [
    UVs {from=MaterialExpressionSwitch_0, output=""}
    Apply View MipBias {from="", output=""}
   ]
   properties {parameterName=BackgroundBaseMap_VTMasks, samplerType=SAMPLERTYPE_VirtualMasks, texture=/UVEditor/Textures/UVEditorColorGrid_VTMask}
  }
  MaterialExpressionTextureSampleParameter2D_17 {
   class=MaterialExpressionTextureSampleParameter2D
   inputs [
    UVs {from=MaterialExpressionSwitch_0, output=""}
    Apply View MipBias {from="", output=""}
   ]
   properties {
    parameterName=BackgroundBaseMap_VTLinearColor
    samplerType=SAMPLERTYPE_VirtualLinearColor
    texture=/UVEditor/Textures/UVEditorColorGrid_VTLinearColor
   }
  }
  MaterialExpressionTextureSampleParameter2D_18 {
   class=MaterialExpressionTextureSampleParameter2D
   inputs [
    UVs {from=MaterialExpressionSwitch_0, output=""}
    Apply View MipBias {from="", output=""}
   ]
   properties {
    parameterName=BackgroundBaseMap_VTLinearGrayscale
    samplerType=SAMPLERTYPE_VirtualLinearGrayscale
    texture=/UVEditor/Textures/UVEditorColorGrid_VTLinearGrayscale
   }
  }
 }
 properties {pixelDepthOffsetMode=PDOM_Legacy, usedShadingModels=Default Lit}
}
